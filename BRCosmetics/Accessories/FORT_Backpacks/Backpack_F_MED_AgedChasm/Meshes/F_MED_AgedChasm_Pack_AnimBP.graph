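# Animation Blueprint for the AgedChasm backpack cosmetic that adds procedural secondary motion to its tail bones.
# The input pose passes through five chained single-body AnimDynamics nodes on alternate bones, dyn_tail_01 to dyn_tail_09, before output.
# Tight angular limits widen toward the tip, angular springs act only on the first two, and component motion is inverted on Y, most strongly at the tip.
# The event graph contains no logic.
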AnimBlueprint /BRCosmetics/Accessories/FORT_Backpacks/Backpack_F_MED_AgedChasm/Meshes/F_MED_AgedChasm_Pack_AnimBP
target_skeleton=&/F_MED_AgedChasm_Pack_Skeleton
graphs [
 AnimGraph {nodes [
   n1 = Anim Dynamics - Bone: dyn_tail_01(ComponentPose=Local To Component(LocalPose=Input Pose()), Alpha=0.8, ExternalForce="0,0,0") {
    angularDampingOverride=1
    angularSpringConstant=20000
    bAngularSpring=true
    bOverrideAngularDamping=true
    bUseGravityOverride=true
    linearSpringConstant=1000
    alphaBoolBlend {blendOption=Linear}
    boundBone {boneName=dyn_tail_01}
    componentLinearAccScale=<1 -2 1>
    componentLinearVelScale=<1 -2 1>
    physicsBodyDefinitions [{
      sphereCollisionRadius=0
      boundBone {boneName=dyn_tail_01}
      boxExtents=<12 15 8>
      constraintSetup {angularLimitsMax=<0 3 5>, angularLimitsMin=<0 -3 -5>, angularTarget=<0 0 0>}
      localJointOffset=<0.5 0 0>
     }]
   }
   n2 = Anim Dynamics - Bone: dyn_tail_03(ComponentPose=@n1, Alpha=1, ExternalForce="0,0,0") {
    angularDampingOverride=1
    angularSpringConstant=20000
    bAngularSpring=true
    bOverrideAngularDamping=true
    bUseGravityOverride=true
    linearSpringConstant=1000
    alphaBoolBlend {blendOption=Linear}
    boundBone {boneName=dyn_tail_03}
    componentLinearAccScale=<1 -2 1>
    componentLinearVelScale=<1 -2 1>
    physicsBodyDefinitions [{
      sphereCollisionRadius=0
      boundBone {boneName=dyn_tail_03}
      boxExtents=<12 15 8>
      constraintSetup {angularLimitsMax=<0 3 5>, angularLimitsMin=<0 -3 -5>, angularTarget=<0 0 0>}
      localJointOffset=<0.5 0 0>
     }]
   }
   n3 = Anim Dynamics - Bone: dyn_tail_05(ComponentPose=@n2, Alpha=1, ExternalForce="0,0,0") {
    angularDampingOverride=1
    angularSpringConstant=20000
    bOverrideAngularDamping=true
    bUseGravityOverride=true
    linearSpringConstant=1000
    alphaBoolBlend {blendOption=Linear}
    boundBone {boneName=dyn_tail_05}
    componentLinearAccScale=<0.4 -2 0.4>
    componentLinearVelScale=<0.4 -2 0.4>
    physicsBodyDefinitions [{
      sphereCollisionRadius=0
      boundBone {boneName=dyn_tail_05}
      boxExtents=<12 15 8>
      constraintSetup {angularLimitsMax=<0 3 5>, angularLimitsMin=<0 -3 -5>, angularTarget=<0 0 0>}
      localJointOffset=<0.5 0 0>
     }]
   }
   n4 = Anim Dynamics - Bone: dyn_tail_07(ComponentPose=@n3, Alpha=1, ExternalForce="0,0,0") {
    angularDampingOverride=1
    angularSpringConstant=20000
    bOverrideAngularDamping=true
    bUseGravityOverride=true
    linearSpringConstant=1000
    alphaBoolBlend {blendOption=Linear}
    boundBone {boneName=dyn_tail_07}
    componentLinearAccScale=<0.4 -2 0.4>
    componentLinearVelScale=<0.4 -2 0.4>
    physicsBodyDefinitions [{
      sphereCollisionRadius=0
      boundBone {boneName=dyn_tail_07}
      boxExtents=<12 15 8>
      constraintSetup {angularLimitsMax=<0 4 5>, angularLimitsMin=<0 -4 -5>, angularTarget=<0 0 0>}
      localJointOffset=<0.5 0 0>
     }]
   }
   n5 = Anim Dynamics - Bone: dyn_tail_09(ComponentPose=@n4, Alpha=1, ExternalForce="0,0,0") {
    angularDampingOverride=1
    angularSpringConstant=20000
    bOverrideAngularDamping=true
    bUseGravityOverride=true
    linearSpringConstant=1000
    alphaBoolBlend {blendOption=Linear}
    boundBone {boneName=dyn_tail_09}
    componentLinearAccScale=<0.4 -5 0.4>
    componentLinearVelScale=<0.4 -5 0.4>
    physicsBodyDefinitions [{
      sphereCollisionRadius=0
      boundBone {boneName=dyn_tail_09}
      boxExtents=<12 15 8>
      constraintSetup {angularLimitsMax=<0 5 5>, angularLimitsMin=<0 -5 -5>, angularTarget=<0 0 0>}
      localJointOffset=<0.5 0 0>
     }]
   }
   Output Pose(Result=Component To Local(ComponentPose=@n5)) {layerGroup=None}
  ]}
 EventGraph {nodes [
   TryGetPawnOwner()
   Event BlueprintUpdateAnimation()
  ]}
]
variables []
